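# Unlit translucent Niagara ribbon material for the CrispRover glider's wave trail.
# Procedural sine stripes on rotated UV1 scroll over time, undulated by a second animated sine, with smoothstep thresholds varying width along the trail.
# A two-color gradient, switched to variant colors by the VariantPicker dynamic parameter, is tinted by particle color and fed through emissive adjustments.
# Opacity combines waves, UV edge fades, particle alpha and depth fade, which mobile skips.
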Material /BRCosmetics/Gadgets/Assets/VinderTech_GliderChute/Glider_CrispRover/FX/M_CrispRoverGlider_WaveTrail
graph {
 properties {bUsedWithNiagaraRibbons=true, blendMode=BLEND_Translucent, pixelDepthOffsetMode=PDOM_Legacy, shadingModel=MSM_Unlit, usedShadingModels=Unlit}
 nodes [
  n1 = DynamicParameter() {parameterIndex=1, defaultValue=<rgba 0 1 1 1>, paramNames [VariantPicker, Param2, Param3, Param4]}
  n2 = NamedRerouteUsage()
  n3 = LinearInterpolate(
   A=LinearInterpolate(
    A=VectorParameter() {parameterName=WaveColor_1, defaultValue=<rgba 0.2865 0.2188 0.0562 0.2>}.RGB
    B=VectorParameter() {parameterName=WaveColorVariant_1, defaultValue=<rgba 0.0538 0.2269 0.3333 0.2>}.RGB
    Alpha=@n1.VariantPicker
   )
   B=LinearInterpolate(
    A=VectorParameter() {parameterName=WaveColor_2, defaultValue=<rgba 0.5833 0.0156 0.0679 1>}.RGB
    B=VectorParameter() {parameterName=WaveColorVariant_2, defaultValue=<rgba 0 0.1081 0.1823 1>}.RGB
    Alpha=@n1.VariantPicker
   )
   Alpha=SmoothStep(Value=ComponentMask(None=@n2) {g=true}) {constMax=0.8, constMin=0.25}
  )
  n4 = ParticleColor()
  MaterialFunctionCall(Emissive=LinearInterpolate(A=Multiply(A=@n3, B=@n4.RGB), B=@n4.RGB, Alpha=ScalarParameter() {parameterName=ParticleColorOnly})) {materialFunction=&MF_FX_EmissiveAdjustments}
  n5 = ScalarParameter() {defaultValue=1, group=WaveSettings, parameterName=WaveBase}
  n6 = ScalarParameter() {defaultValue=0.1, group=WaveSettings, parameterName=TipWidthFactor}
  n7 = ScalarParameter() {defaultValue=0.4, group=WaveSettings, parameterName=WaveWidthMax}
  n8 = ComponentMask(None=NamedRerouteUsage()) {g=true}
  n9 = NamedRerouteUsage()
  n10 = ComponentMask(None=@n9) {g=true}
  MaterialExpressionRemap_1 = MaterialXRemap!(None=Multiply(A=@n10, B=@n10), InputLow, InputHigh, TargetLow=Constant() {r=0.1}, TargetHigh) {inputHighDefault=0, inputLowDefault=1}
  n11 = ComponentMask(None=@n9) {r=true}
  n12 = ComponentMask(None=AppendVector(A=SmoothStep(Max=@MaterialExpressionRemap_1, Value=@n11), B=ComponentMask(None=@n9) {g=true})) {r=true}
  n13 = LinearInterpolate(
   A=LinearInterpolate(A=Multiply(A=@n6, B=@n7), B=@n7, Alpha=SmoothStep(Value=@n8) {constMax=0.8})
   B=ScalarParameter() {group=WaveSettings, parameterName=WaveWidthMin}
   Alpha=Saturate(None=SmoothStep(Value=@n12) {constMin=0.5})
  )
  n14 = ScalarParameter() {defaultValue=3, group=WaveSettings, parameterName=WavePatternSizeTip}
  n15 = ScalarParameter() {defaultValue=6, group=WaveSettings, parameterName=WavePatternSize}
  n16 = ComponentMask(None=NamedRerouteUsage()) {r=true}
  n17 = Add(
   A=Multiply(A=Time(), B=ScalarParameter() {defaultValue=0.5, group=WaveSettings, parameterName=WaveSpeedMin})
   B=Multiply(A=NamedRerouteUsage(), B=LinearInterpolate(A=@n14, B=@n15, Alpha=@n16))
  )
  n18 = ComponentMask(None=NamedRerouteUsage()) {r=true}
  n19 = LinearInterpolate(
   A=ScalarParameter() {defaultValue=0.4, group=WaveSettings, parameterName=WaveOndulationFrequencyMin}
   B=ScalarParameter() {defaultValue=1, group=WaveSettings, parameterName=WaveOndulationFrequencyMax}
   Alpha=ComponentMask(None=NamedRerouteUsage()) {r=true}
  )
  n20 = Time()
  n21 = MaterialFunctionCall(X=ScalarParameter() {defaultValue=-2, group=WaveSettings, parameterName=OndulationSpeedMax}) {materialFunction=&MakeFloat2}
  n22 = Sine(None=Add(
    A=MaterialFunctionCall(UVs=@n18, Texture Scale=@n19) {materialFunction=&ScaleUVsByCenter}.UVs
    B=Multiply(A=@n20, B=@n21.Result)
   ) {constB=0.1})
  n23 = LinearInterpolate(
   A=ScalarParameter() {defaultValue=1, group=WaveSettings, parameterName=WaveOndulationIntensityMin}
   B=ScalarParameter() {group=WaveSettings, parameterName=WaveOndulationIntensityMax}
   Alpha=ComponentMask(None=NamedRerouteUsage()) {r=true}
  )
  n24 = Sine(None=Add(A=ComponentMask(None=@n17) {g=true}, B=Multiply(A=@n22, B=@n23) {constB=0.2}))
  n25 = SmoothStep(Min=Subtract(A=@n5, B=@n13), Max=Add(A=@n5, B=@n13), Value=Saturate(None=ConstantBiasScale(None=@n24))) {constMax=0.95, constMin=0.9}
  n26 = TextureCoordinate() {coordinateIndex=1}
  n27 = ComponentMask(None=@n26) {r=true}
  n28 = OneMinus(None=ComponentMask(None=@n26) {g=true})
  n29 = Multiply(
   A=Multiply(A=SmoothStep(Value=@n27) {constMax=0.2}, B=SmoothStep(Value=@n28) {constMax=0.3})
   B=SmoothStep(Value=ComponentMask(None=@n26) {g=true}) {constMax=0.3}
  )
  n30 = DepthFade(FadeDistance=ScalarParameter() {defaultValue=50, parameterName=FadeDistance})
  Multiply(
   A=Multiply(A=@n4.A, B=Multiply(A=@n25, B=@n29))
   B=ShadingPathSwitch(Default=@n30, Deferred=@n30, Mobile=Constant() {r=1}) {desc=Migrated from FeatureLevelSwitch}
  )
  n31 = TextureCoordinate() {coordinateIndex=1}
  n32 = ScalarParameter() {defaultValue=0.3, group=WaveSettings, parameterName=RotateWaves}
  NamedRerouteDeclaration(None=Saturate(None=Rotator(Coordinate=@n31, Time=@n32) {speed=1})) {name=WavePatternUVs, nodeColor=<rgba 1 0.6328 0 1>}
 ]
}
