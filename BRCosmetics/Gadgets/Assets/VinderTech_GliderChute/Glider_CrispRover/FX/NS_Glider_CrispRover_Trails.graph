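# Niagara system for CrispRover glider trails: large and small motion and side ribbons, plus a disabled speedlines emitter.
# Each ribbon emitter spawns two AnimTrails on skeletal-mesh sockets, pushed along Y and offset by owner velocity.
# Motion trails start after User.Delay and blend color by User.VariantPicker.
# A system timeline scales width and opacity, with alpha also driven by an integrator, time of day and front-end versus gameplay overrides.
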NiagaraSystem /BRCosmetics/Gadgets/Assets/VinderTech_GliderChute/Glider_CrispRover/FX/NS_Glider_CrispRover_Trails
dependencies {
 usedDataInterfaces [&NiagaraDataInterfaceCurve, &NiagaraDataInterfaceVector2DCurve, &NiagaraDataInterfaceColorCurve]
 usedDynamicInputs [&FN_Select_Color_from_Time_Of_Day]
 usedModules [
  &SystemState
  &FN_Integrator
  &self:SystemSpawnScript.SystemScriptSource.SystemScriptGraph.NiagaraNodeAssignment_1.SetVariables_#1
  &Timeline
 ]
}
emitters {
 MotionTrailsLarge {
  emitterSpawnScript [&FN_RibbonIDOffsetUtility()]
  emitterUpdateScript [
   &EmitterState(Life Cycle Mode=::NewEnumerator1, Loop Delay=&Add_Float(A=$User.Delay, B=0.05), Loop Duration=1, UseLoopDelay=true)
   &FN_SpawnAnimTrail(Number of Anim Trails=2)
  ]
  particleSpawnScript [
   &FN_InitializeAnimTrailBoneIndecies()
   &FN_InitializeAnimTrailBoneIndecies(AnimTrail Selection=1, First Bone or Socket Index=2, Second Bone or Socket Index=3) {module=FN_InitializeAnimTrailBoneIndecies001}
   &FN_InitializeAnimTrail(
    AnimTrailWidth=&FN_FrontEndFloatModifier(Gameplay=1, Front End=0.75)
    Color=&LerpLinearColors(StartColor=<rgba 0.1154 0.0299 0.7031 1>, EndColor=<rgba 0.5823 0.6797 0.7031 1>, LerpFactor=$User.VariantPicker)
    Lifetime=0.15
    Write Material Random=false
   )
   /SetVariables(Particles.WaveColor:LinearColor=<rgba 1 1 1 2>) {isSetParameters=true}
   &AddVelocity(Velocity=&FN_FrontEndVector3Modifier(Gameplay=<0 1500 0>, Front End=<0 1500 0>))
   &CameraOffset(Camera Offset Amount=-15)
   &DynamicMaterialParameters(
    Index 1 Param 1=$User.VariantPicker
    Module.Param0WriteEnabled1=true
    Module.Param1WriteEnabled1=true
    Module.Param2WriteEnabled1=true
    Module.Param3WriteEnabled1=true
    Write Parameter Index 0=false
    Write Parameter Index 1=true
   ) {module=DynamicMaterialParameters001}
  ]
  particleUpdateScript [
   &InheritSourceMovement(
    Applied Position Scale=&Multiply_VectorByFloat(
     Vector=<-0.8 -0.8 0>
     Float=&FloatFromCurve(FloatCurve=DataInterface() {propertyValues {Curve {keys [
          {interpMode=RCIM_Cubic, tangentMode=RCTM_User, time=0.5, value=1}
          {arriveTangent=0.0051, interpMode=RCIM_Cubic, leaveTangent=0.0051, tangentMode=RCTM_User, time=1, value=0}
         ]}}})
    )
    ApplyForce=false
    ApplyPosition=true
    ApplyPositionFalloff=false
    Source Velocity:Vector3f=&TransformVector(Vector=$Engine.Owner.Velocity, Source Space=::World, Destination Space=::Local)
   )
   &SolveForcesAndVelocity(Clamp Velocity=false, Limit Acceleration=false)
   &FN_UpdateAnimTrail(
    Locked Particle RibbonWidth:NiagaraFloat=&Multiply_Float(A=$Particles.Initial.RibbonWidth, B=$System.Timeline.NormalizedTime)
    SkeletalMesh=$User.SkeletalMesh
   ) {module=FN_UpdateAnimTrail001}
   /SetVariables(Particles.RibbonU0Override=$Particles.NormalizedAge) {isSetParameters=true, module=SetVariables_#2}
   /SetVariables(Particles.WaveColor:LinearColor=&MultiplyLinearColor(
     Linear Color=&MultiplyLinearColor(
      Linear Color=$Particles.Initial.WaveColor
      Scale Factor=&MakeVector4(
       X=1
       Y=1
       Z=1
       W=&Multiply_Float(
        A=&Multiply_Float(
         A=&Multiply_Float(
          A=&Multiply_Float(
           A=&Max_Float(
            A:NiagaraFloat=0
            B:NiagaraFloat=&FN_FrontEndFloatModifier(Gameplay=$System.FN_Integrator.Value, Front End=1)
           )
           B=&FloatFromCurve(FloatCurve=DataInterface() {propertyValues {Curve {keys [
                {arriveTangent=-1.0608, interpMode=RCIM_Cubic, leaveTangent=-1.0608, tangentMode=RCTM_User, time=0, value=1}
                {arriveTangent=-0.6176, interpMode=RCIM_Cubic, leaveTangent=-0.6176, tangentMode=RCTM_User, time=1, value=0}
               ]}}})
          )
          B=&FNTimeOfDayScalar(Night Scale=1, Morning Scale=1, Evening Scale=1)
         )
         B=&FloatFromCurve(FloatCurve=DataInterface() {propertyValues {Curve {keys [<at 0 1>,<at 1 0>]}}})
        )
        B=$System.Timeline.NormalizedTime
       )
      )
     )
     Scale Factor=&MakeVector4_FromLinearColor(LinearColor=$System.TimeOfDayMultiplier)
    )) {isSetParameters=true}
   &ScaleColor(
    Color Value To Scale=$Particles.Color
    Scale Alpha=&Multiply_Float(
     A=&Multiply_Float(
      A=&FN_FrontEndFloatModifier(
       Gameplay=&Multiply_Float(
        A=&Max_Float(A:NiagaraFloat=0.6, B:NiagaraFloat=$System.FN_Integrator.Value)
        B=&FloatFromCurve(
         FloatCurve=DataInterface() {propertyValues {Curve {keys [{interpMode=RCIM_Cubic, tangentMode=RCTM_User, time=0, value=1}, {interpMode=RCIM_Cubic, tangentMode=RCTM_User, time=0.2, value=0}]}}}
         Scale Curve=0.4
        )
       )
       Front End=0.4
      )
      B=&FNTimeOfDayScalar(Night Scale=0.4, Morning Scale=0.5, Evening Scale=0.5)
     )
     B=$System.Timeline.NormalizedTime
    )
    ScaleA=true
    ScaleRGB=true
   ) {module=ScaleColor001}
   &DynamicMaterialParameters(
    Index 0 Param 4=&FN_FrontEndFloatModifier(Gameplay=$System.FN_Integrator.Value, Front End=1)
    Module.Param0WriteEnabled=true
    Module.Param1WriteEnabled=true
    Module.Param2WriteEnabled=true
    Module.Param3WriteEnabled=true
   )
  ]
  rendererClasses [&NiagaraRibbonRendererProperties]
  renderers [
   {rendererClass=&NiagaraRibbonRendererProperties}
   {rendererClass=&NiagaraRibbonRendererProperties, rendererIndex=1}
  ]
 }
 MotionTrailsSmall {
  emitterSpawnScript [&FN_RibbonIDOffsetUtility()]
  emitterUpdateScript [
   &EmitterState(Life Cycle Mode=::NewEnumerator1, Loop Delay=&Add_Float(A=$User.Delay, B=0.05), Loop Duration=1, UseLoopDelay=true)
   &FN_SpawnAnimTrail(Number of Anim Trails=2)
  ]
  particleSpawnScript [
   &FN_InitializeAnimTrailBoneIndecies(First Bone or Socket Index=4, Second Bone or Socket Index=5)
   &FN_InitializeAnimTrailBoneIndecies(AnimTrail Selection=1, First Bone or Socket Index=6, Second Bone or Socket Index=7) {module=FN_InitializeAnimTrailBoneIndecies001}
   &FN_InitializeAnimTrail(
    AnimTrailWidth=&FN_FrontEndFloatModifier(Gameplay=1, Front End=0.75)
    Color=&LerpLinearColors(StartColor=<rgba 0.1154 0.0299 0.7031 1>, EndColor=<rgba 0.5823 0.6797 0.7031 1>, LerpFactor=$User.VariantPicker)
    Lifetime=0.1
    Write Material Random=false
   )
   /SetVariables(Particles.WaveColor:LinearColor=<rgba 1 1 1 2>) {isSetParameters=true}
   &AddVelocity(Velocity=&FN_FrontEndVector3Modifier(Gameplay=<0 1500 0>, Front End=<0 1500 0>))
   &CameraOffset(Camera Offset Amount=-15)
   &DynamicMaterialParameters(
    Index 1 Param 1=$User.VariantPicker
    Module.Param0WriteEnabled1=true
    Module.Param1WriteEnabled1=true
    Module.Param2WriteEnabled1=true
    Module.Param3WriteEnabled1=true
    Write Parameter Index 0=false
    Write Parameter Index 1=true
   )
  ]
  particleUpdateScript [
   &InheritSourceMovement(
    Applied Position Scale=&Multiply_VectorByFloat(
     Vector=<-0.8 -0.8 0>
     Float=&FloatFromCurve(FloatCurve=DataInterface() {propertyValues {Curve {keys [
          {interpMode=RCIM_Cubic, tangentMode=RCTM_User, time=0.5, value=1}
          {arriveTangent=0.0051, interpMode=RCIM_Cubic, leaveTangent=0.0051, tangentMode=RCTM_User, time=1, value=0}
         ]}}})
    )
    ApplyForce=false
    ApplyPosition=true
    ApplyPositionFalloff=false
    Source Velocity:Vector3f=&TransformVector(Vector=$Engine.Owner.Velocity, Source Space=::World, Destination Space=::Local)
   )
   &SolveForcesAndVelocity(Clamp Velocity=false, Limit Acceleration=false)
   &FN_UpdateAnimTrail(
    Locked Particle RibbonWidth:NiagaraFloat=&Multiply_Float(A=$Particles.Initial.RibbonWidth, B=$System.Timeline.NormalizedTime)
    SkeletalMesh=$User.SkeletalMesh
   ) {module=FN_UpdateAnimTrail001}
   /SetVariables(Particles.RibbonU0Override=$Particles.NormalizedAge) {isSetParameters=true, module=SetVariables_#2}
   /SetVariables(Particles.WaveColor:LinearColor=&MultiplyLinearColor(
     Linear Color=&MultiplyLinearColor(
      Linear Color=$Particles.Initial.WaveColor
      Scale Factor=&MakeVector4(
       X=1
       Y=1
       Z=1
       W=&Multiply_Float(
        A=&Multiply_Float(
         A=&Multiply_Float(
          A=&Multiply_Float(
           A=&Max_Float(
            A:NiagaraFloat=0
            B:NiagaraFloat=&FN_FrontEndFloatModifier(Gameplay=$System.FN_Integrator.Value, Front End=1)
           )
           B=&FloatFromCurve(FloatCurve=DataInterface() {propertyValues {Curve {keys [
                {arriveTangent=-1.0608, interpMode=RCIM_Cubic, leaveTangent=-1.0608, tangentMode=RCTM_User, time=0, value=1}
                {arriveTangent=-0.6176, interpMode=RCIM_Cubic, leaveTangent=-0.6176, tangentMode=RCTM_User, time=1, value=0}
               ]}}})
          )
          B=&FNTimeOfDayScalar(Night Scale=1, Morning Scale=1, Evening Scale=1)
         )
         B=&FloatFromCurve(FloatCurve=DataInterface() {propertyValues {Curve {keys [<at 0 1>,<at 1 0>]}}})
        )
        B=$System.Timeline.NormalizedTime
       )
      )
     )
     Scale Factor=&MakeVector4_FromLinearColor(LinearColor=$System.TimeOfDayMultiplier)
    )) {isSetParameters=true}
   &ScaleColor(
    Color Value To Scale=$Particles.Color
    Scale Alpha=&Multiply_Float(
     A=&Multiply_Float(
      A=&FN_FrontEndFloatModifier(
       Gameplay=&Multiply_Float(
        A=&Max_Float(A:NiagaraFloat=0.6, B:NiagaraFloat=$System.FN_Integrator.Value)
        B=&FloatFromCurve(
         FloatCurve=DataInterface() {propertyValues {Curve {keys [{interpMode=RCIM_Cubic, tangentMode=RCTM_User, time=0, value=1}, {interpMode=RCIM_Cubic, tangentMode=RCTM_User, time=0.2, value=0}]}}}
         Scale Curve=0.4
        )
       )
       Front End=0.4
      )
      B=&FNTimeOfDayScalar(Night Scale=0.4, Morning Scale=0.5, Evening Scale=0.5)
     )
     B=$System.Timeline.NormalizedTime
    )
    ScaleA=true
    ScaleRGB=true
   ) {module=ScaleColor001}
  ]
  rendererClasses [&NiagaraRibbonRendererProperties]
  renderers [
   {rendererClass=&NiagaraRibbonRendererProperties}
   {rendererClass=&NiagaraRibbonRendererProperties, rendererIndex=1}
  ]
 }
 SideTrailLarge {
  emitterSpawnScript [&FN_RibbonIDOffsetUtility()]
  emitterUpdateScript [
   &EmitterState(Life Cycle Mode=::NewEnumerator1, Loop Duration=1, UseLoopDelay=false)
   &FN_SpawnAnimTrail(Number of Anim Trails=2)
  ]
  particleSpawnScript [
   &FN_InitializeAnimTrailBoneIndecies(First Bone or Socket Index=1)
   &FN_InitializeAnimTrailBoneIndecies(AnimTrail Selection=1, First Bone or Socket Index=3, Second Bone or Socket Index=3) {module=FN_InitializeAnimTrailBoneIndecies001}
   &FN_InitializeAnimTrail(AnimTrailWidth=4, Color=<rgba 0.5885 0.5885 0.5885 1>, Lifetime=0.25)
   &AddVelocity(Velocity=&FN_FrontEndVector3Modifier(Gameplay=<0 3500 0>, Front End=<0 4500 0>)) {module=AddVelocity001}
  ]
  particleUpdateScript [
   &InheritSourceMovement(
    Applied Position Scale=&Multiply_VectorByFloat(
     Vector=<-1 -1 0>
     Float=&FloatFromCurve(FloatCurve=DataInterface() {propertyValues {Curve {keys [
          {interpMode=RCIM_Cubic, tangentMode=RCTM_User, time=0.5, value=1}
          {arriveTangent=0.0051, interpMode=RCIM_Cubic, leaveTangent=0.0051, tangentMode=RCTM_User, time=1, value=0}
         ]}}})
    )
    ApplyForce=false
    ApplyPosition=true
    ApplyPositionFalloff=false
    Source Velocity:Vector3f=&TransformVector(Vector=$Engine.Owner.Velocity, Source Space=::World, Destination Space=::Local)
   )
   &SolveForcesAndVelocity(Clamp Velocity=false, Limit Acceleration=false)
   &FN_UpdateAnimTrail(
    Locked Particle RibbonWidth:NiagaraFloat=&Multiply_Float(A=$Particles.Initial.RibbonWidth, B=$System.Timeline.NormalizedTime)
    SkeletalMesh=$User.SkeletalMesh
   ) {module=FN_UpdateAnimTrail001}
   /SetVariables(Particles.RibbonU0Override=$Particles.NormalizedAge) {isSetParameters=true, module=SetVariables_#2}
   &ScaleColor(
    Color Value To Scale=$Particles.Color
    Scale Alpha=&Multiply_Float(
     A=&Multiply_Float(
      A=&FN_FrontEndFloatModifier(
       Gameplay=&Multiply_Float(
        A=&Max_Float(A:NiagaraFloat=0.4, B:NiagaraFloat=$System.FN_Integrator.Value)
        B=&FloatFromCurve(
         FloatCurve=DataInterface() {propertyValues {Curve {keys [
             {arriveTangent=-3.3328, interpMode=RCIM_Cubic, leaveTangent=-3.3328, tangentMode=RCTM_User, time=0, value=0.5}
             {interpMode=RCIM_Cubic, tangentMode=RCTM_User, time=0.5, value=0}
            ]}}}
         Scale Curve=0.1
        )
       )
       Front End=0.1
      )
      B=&FNTimeOfDayScalar(Night Scale=0.4, Morning Scale=0.5, Evening Scale=0.5)
     )
     B=$System.Timeline.NormalizedTime
    )
    ScaleA=true
    ScaleRGB=true
   )
   &DynamicMaterialParameters(
    Index 0 Param 1=&FloatFromCurve(FloatCurve=DataInterface() {propertyValues {Curve {keys [
         {interpMode=RCIM_Cubic, tangentMode=RCTM_User, time=0, value=1}
         {arriveTangent=-2.3194, interpMode=RCIM_Cubic, leaveTangent=-2.3194, tangentMode=RCTM_User, time=1, value=0}
        ]}}})
    Index 0 Param 3=-5
    Module.Param0WriteEnabled=true
    Module.Param1WriteEnabled=true
    Module.Param2WriteEnabled=true
    Module.Param3WriteEnabled=true
   )
  ]
  rendererClasses [&NiagaraRibbonRendererProperties]
  renderers [{rendererClass=&NiagaraRibbonRendererProperties}]
 }
 SideTrailSmall {
  emitterSpawnScript [&FN_RibbonIDOffsetUtility()]
  emitterUpdateScript [
   &EmitterState(Life Cycle Mode=::NewEnumerator1, Loop Duration=1, UseLoopDelay=false)
   &FN_SpawnAnimTrail(Number of Anim Trails=2)
  ]
  particleSpawnScript [
   &FN_InitializeAnimTrailBoneIndecies(First Bone or Socket Index=5, Second Bone or Socket Index=5)
   &FN_InitializeAnimTrailBoneIndecies(AnimTrail Selection=1, First Bone or Socket Index=7, Second Bone or Socket Index=7) {module=FN_InitializeAnimTrailBoneIndecies001}
   &FN_InitializeAnimTrail(AnimTrailWidth=4, Lifetime=0.05)
   &AddVelocity(Velocity=&FN_FrontEndVector3Modifier(Gameplay=<0 3500 0>, Front End=<0 4500 0>)) {module=AddVelocity001}
  ]
  particleUpdateScript [
   &InheritSourceMovement(
    Applied Position Scale=&Multiply_VectorByFloat(
     Vector=<-1 -1 0>
     Float=&FloatFromCurve(FloatCurve=DataInterface() {propertyValues {Curve {keys [
          {interpMode=RCIM_Cubic, tangentMode=RCTM_User, time=0.5, value=1}
          {arriveTangent=0.0051, interpMode=RCIM_Cubic, leaveTangent=0.0051, tangentMode=RCTM_User, time=1, value=0}
         ]}}})
    )
    ApplyForce=false
    ApplyPosition=true
    ApplyPositionFalloff=false
    Source Velocity:Vector3f=&TransformVector(Vector=$Engine.Owner.Velocity, Source Space=::World, Destination Space=::Local)
   )
   &SolveForcesAndVelocity(Clamp Velocity=false, Limit Acceleration=false)
   &FN_UpdateAnimTrail(
    Locked Particle RibbonWidth:NiagaraFloat=&Multiply_Float(A=$Particles.Initial.RibbonWidth, B=$System.Timeline.NormalizedTime)
    SkeletalMesh=$User.SkeletalMesh
   ) {module=FN_UpdateAnimTrail001}
   /SetVariables(Particles.RibbonU0Override=$Particles.NormalizedAge) {isSetParameters=true, module=SetVariables_#2}
   &ScaleColor(
    Color Value To Scale=$Particles.Color
    Scale Alpha=&Multiply_Float(
     A=&Multiply_Float(
      A=&FN_FrontEndFloatModifier(
       Gameplay=&Multiply_Float(
        A=&Max_Float(A:NiagaraFloat=0.4, B:NiagaraFloat=$System.FN_Integrator.Value)
        B=&FloatFromCurve(
         FloatCurve=DataInterface() {propertyValues {Curve {keys [
             {arriveTangent=-3.3328, interpMode=RCIM_Cubic, leaveTangent=-3.3328, tangentMode=RCTM_User, time=0, value=0.5}
             {interpMode=RCIM_Cubic, tangentMode=RCTM_User, time=0.5, value=0}
            ]}}}
         Scale Curve=0.1
        )
       )
       Front End=0.1
      )
      B=&FNTimeOfDayScalar(Night Scale=0.4, Morning Scale=0.5, Evening Scale=0.5)
     )
     B=$System.Timeline.NormalizedTime
    )
    ScaleA=true
    ScaleRGB=true
   )
   &DynamicMaterialParameters(
    Index 0 Param 1=&FloatFromCurve(FloatCurve=DataInterface() {propertyValues {Curve {keys [
         {interpMode=RCIM_Cubic, tangentMode=RCTM_User, time=0, value=1}
         {arriveTangent=-2.3194, interpMode=RCIM_Cubic, leaveTangent=-2.3194, tangentMode=RCTM_User, time=1, value=0}
        ]}}})
    Index 0 Param 3=-5
    Module.Param0WriteEnabled=true
    Module.Param1WriteEnabled=true
    Module.Param2WriteEnabled=true
    Module.Param3WriteEnabled=true
   )
  ]
  rendererClasses [&NiagaraRibbonRendererProperties]
  renderers [{rendererClass=&NiagaraRibbonRendererProperties}]
 }
 Speedlines {
  bEnabled=false
  emitterUpdateScript [
   &EmitterState(Enable Distance Culling=true, MaxDistance=2048, Module.bUseMaxDistance=true, Module.bUseMinDistance=false, Scalability Mode=::NewEnumerator1)
   &SpawnRate(
    SpawnRate=&Multiply_Float(
     A=&FN_FrontEndFloatModifier(Gameplay=$User.ForwardDot, Front End=1)
     B=&UniformRangedFloat(Minimum=75, Maximum=150)
    )
    Use Spawn Probability=false
   )
  ]
  particleSpawnScript [
   /Niagara/Modules/Spawn/Initialization/InitializeParticle(
    Color=&MakeLinearColorFromVectorAndFloat(
     "Vector (RGB)"=<1 1 1>
     "Float (Alpha)"=&Multiply_Float(
      A=&UniformRangedFloat(Minimum=0.015, Maximum=0.15)
      B=&Add_Float(
       A=1
       B=&Multiply_Float(A=10, B=&Subtract_Float(A=1, B=$Engine.Emitter.SpawnCountScale))
      )
     )
    )
    Lifetime=&FN_FrontEndFloatModifier(
     Gameplay=&UniformRangedFloat(Minimum=0.2, Maximum=0.35)
     Front End=&UniformRangedFloat(Minimum=0.1, Maximum=0.2)
    )
    Module.Write Color=true
    Module.Write Lifetime=true
    Module.Write Mass=true
    Module.Write Material Random=true
    Module.Write Position=false
    Module.Write Scale=false
    Module.Write SpriteRotation=false
    Module.Write SpriteSize=true
    Position=&SimulationPosition()
    Sprite Size=&UniformRangedVector2D(Minimum=<3 4>, Maximum=<5 6>)
   )
   &CylinderLocation(
    Cylinder Height=200
    Cylinder Radius=&UniformRangedFloat(Minimum=175, Maximum=225)
    MaskSpawnGroup=false
    Module.Enable Lathe Profile=false
    Module.Override Randomness=false
    Module.Surface Only=true
    Orientation Axis:ENiagaraOrientationAxis=::XAxis
    Use Endcaps In Surface Only Mode=false
   )
   &AddVelocity(Velocity=&FN_FrontEndVector3Modifier(Gameplay=<0 1000 0>, Front End=<0 3000 0>))
   &AddVelocityFromPoint(
    Module.Use Falloff Distance=false
    Velocity Strength=&Multiply_Float(
     A=&FN_FrontEndFloatModifier(Gameplay=1, Front End=2)
     B=&UniformRangedFloat(Minimum=500, Maximum=700)
    )
   )
  ]
  particleUpdateScript [
   &ParticleState()
   &SolveForcesAndVelocity(Clamp Velocity=false, Limit Acceleration=false)
   &ScaleSpriteSize(Scale Factor=&Vector2DFromCurve(
     Vector2Curve=DataInterface() {propertyValues {XCurve {keys [<at 0 0.5 RCIM_Cubic>,<at 0.5 1 RCIM_Cubic>]}, YCurve {keys [<at 0 0 RCIM_Cubic>,<at 0.5 1 RCIM_Cubic>]}}}
     ScaleCurve=&Vector2DFromFloat(Value=1)
    ))
   &ScaleSpriteSizeBySpeed(Max Scale Factor=<1 40>, Min Scale Factor=<0.01 0.05>, Velocity Threshold=1)
   &ScaleColor(
    Scale Alpha=&FloatFromCurve(FloatCurve=DataInterface() {propertyValues {Curve {keys [<at 0 0 RCIM_Cubic>, <at 0.2 1>, {arriveTangent=-2.221, interpMode=RCIM_Cubic, leaveTangent=-2.221, tangentMode=RCTM_User, time=1, value=0}]}}})
    ScaleA=true
    ScaleRGB=true
   )
  ]
  rendererClasses [&NiagaraSpriteRendererProperties]
  renderers [{rendererClass=&NiagaraSpriteRendererProperties}]
 }
}
summary {
 systemName=NS_Glider_CrispRover_Trails
 emitters [
  {bEnabled=false, emitterName=Speedlines, rendererClasses [&NiagaraSpriteRendererProperties]}
  {emitterName=MotionTrailsLarge, rendererClasses [&NiagaraRibbonRendererProperties]}
  {emitterName=SideTrailLarge, rendererClasses [&NiagaraRibbonRendererProperties]}
  {emitterName=MotionTrailsSmall, rendererClasses [&NiagaraRibbonRendererProperties]}
  {emitterName=SideTrailSmall, rendererClasses [&NiagaraRibbonRendererProperties]}
 ]
 userVariables [
  User.bIsFrontEnd {type=NiagaraBool, defaultValue=true}
  User.bIsFrontEndPreview {type=NiagaraBool}
  User.bIsFullyDeployed {type=NiagaraBool}
  User.BPOpacityMult {type=NiagaraFloat, defaultValue=0, description=""}
  User.Color {type=LinearColor, description="", defaultValue {struct=LinearColor, value=<rgba 1 1 1 1>}}
  User.CurrentExhaust {type=Vector3f, description="", defaultValue {struct=Vector3f, value=<0 0 0>}}
  User.Delay {type=NiagaraFloat, defaultValue=0, description=""}
  User.ExhaustAngle {type=NiagaraFloat, defaultValue=0, description=""}
  User.ForwardDot {type=NiagaraFloat, defaultValue=0, description=""}
  User.LocalAlphaRadial {type=NiagaraFloat, defaultValue=0.25, description=""}
  User.PlayerSpeed {type=NiagaraFloat, defaultValue=0, description=""}
  User.RibbonIDOffset {type=NiagaraInt32}
  User.RightDot {type=NiagaraFloat, defaultValue=0, description=""}
  User.SizeScale {type=NiagaraFloat, defaultValue=0, description=""}
  User.SkeletalMesh {
   type=NiagaraDataInterfaceSkeletalMesh
   defaultValue {
    struct=NiagaraExt_VariableValue_DataInterface
    value {dataInterface=&self:NiagaraDataInterfaceSkeletalMesh_0}
   }
  }
  User.VariantPicker {type=NiagaraFloat, defaultValue=0, description=""}
 ]
}
